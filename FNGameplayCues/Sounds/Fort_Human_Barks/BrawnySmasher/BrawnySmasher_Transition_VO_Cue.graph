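# Sound cue for the BrawnySmasher NPC's transition voice bark, using MotherWall "TransitionToAlien" recordings.
# A Random node at the root picks one of three equally weighted wave players and is excluded from branch culling.
# It plays through the NPC VO sound class with BrawnySmasher attenuation and a boosted volume multiplier.
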SoundCue /FNGameplayCues/Sounds/Fort_Human_Barks/BrawnySmasher/BrawnySmasher_Transition_VO_Cue
root=Random_1
nodes [
 {id=Random_1, children [WavePlayer_0,WavePlayer_1,WavePlayer_2], properties {bShouldExcludeFromBranchCulling=true, weights [1,1,1]}}
 {id=WavePlayer_0, properties {soundWaveAssetPtr=&/MotherWall_2021_TransitionToAlien_010}}
 {id=WavePlayer_1, properties {soundWaveAssetPtr=&/MotherWall_2021_TransitionToAlien_011}}
 {id=WavePlayer_2, properties {soundWaveAssetPtr=&/MotherWall_2021_TransitionToAlien_012}}
]
properties {
 attenuationSettings=/Game/Athena/Sounds/NPCs/BrawnySmasher/BrawnySmasher_Att
 maxDistance=4000
 soundClassObject=&Fort_VO_NPC
 volumeMultiplier=1.7
 concurrencySet [null]
}
